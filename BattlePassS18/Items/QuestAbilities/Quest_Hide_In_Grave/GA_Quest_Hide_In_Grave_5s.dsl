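; Quest gameplay ability that credits the player for staying hidden in a qualifying prop, by its name a grave for 5 seconds.
; A Hide event starts a timer if the instigating B_HidingProp has a target tag, otherwise the ability ends.
; Timer completion sends a custom stat event to the quest manager and ends the ability.
; StopHiding clears any running timer and ends the ability without credit.
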(event Custom|OnHideTimerComplete
  ;[0.3495 0.1005 1] Initiate hiding timer...
  (Quest|SendComplexCustomStatEvent (Variables|GAVars|Data|GetOwningPlayerQuestManager) 0 (Variables|Tags|GetHidingQuestTag) "(GameplayTags=)" 1)
  (Development|PrintString "@@@@@@@@@ ActivateAbilityFromEvent [GA_Quest_Hide_In_Grave_5s]: 5 seconds hidden = Quest success!" false (Variables|Debug|GetbDebugAbility))
  (Ability|EndAbility))

(event Ability|EventActivateAbilityFromEvent
  (params (EventData :type Gameplay%20Event%20Data%20Structure%20%28by%20ref%29))
  (bind (_eventtag _instigator) (Utilities|Struct|BreakGameplayEventData EventData))
  (|CallParentFunction EventData)
  (Development|PrintString "@@@ ActivateAbilityFromEvent [GA_Quest_Hide_In_Grave_5s]: CALLED!" false (Variables|Debug|GetbDebugAbility))
  (Variables|EventData|SetInstigator _instigator)
  (switch Utilities|FlowControl|Switch|SwitchonGameplayTag _eventtag
    (:Event.Athena.HidingProp.Hide
      ;[0.03372 0.8198 1] Start quest timer if the hide spot meets target requirements...
      (Development|PrintString "@@@@@@ ActivateAbilityFromEvent [GA_Quest_Hide_In_Grave_5s]: Hide..." false (Variables|Debug|GetbDebugAbility))
      (bind _asb_hiding_prop (Utilities|Casting|CastToB_HidingProp (Variables|EventData|GetInstigator))
        (:then
          (if (GameplayTags|HasAnyMatchingGameplayTags _asb_hiding_prop (Variables|Tags|GetTargetHidingPropTags))
            (bind _returnvalue (Utilities|Time|SetTimerbyEvent (delegate Custom|OnHideTimerComplete) (Variables|AbilityData|GetHidingTime)))
            (Variables|AbilityData|SetHidingTimerHandle _returnvalue)
            (else
              (Ability|EndAbility))))
        (:CastFailed
          (Ability|EndAbility))))
    (:Event.Athena.HidingProp.StopHiding
      ;[1 0.1009 0.7471] Player left grave early...
      (Development|PrintString "@@@@@@ ActivateAbilityFromEvent [GA_Quest_Hide_In_Grave_5s]: Stop Hiding..." false (Variables|Debug|GetbDebugAbility))
      (if (Utilities|Time|IsValidTimerHandle (Variables|AbilityData|GetHidingTimerHandle))
        (Development|PrintString "@@@@@@@@@ ActivateAbilityFromEvent [GA_Quest_Hide_In_Grave_5s]: grave exited - ending timer..." false (Variables|Debug|GetbDebugAbility))
        (Utilities|Time|ClearandInvalidateTimerbyHandle (Variables|AbilityData|GetHidingTimerHandle))
        (else
          _))
      (Ability|EndAbility))))
